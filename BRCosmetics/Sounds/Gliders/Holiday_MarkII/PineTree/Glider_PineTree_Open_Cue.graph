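# Sound cue for the opening sound of the holiday PineTree glider cosmetic.
# A short fixed delay feeds a single wave player that plays the PineTree open sound.
# Playback uses shared glider open/close attenuation and concurrency, the cosmetic glider-open property sheet, and the glider landing sound class at slightly reduced volume.
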SoundCue /BRCosmetics/Sounds/Gliders/Holiday_MarkII/PineTree/Glider_PineTree_Open_Cue
root=Delay_0
nodes [
 {id=Delay_0, children [WavePlayer_0], properties {delayMax=0.13, delayMin=0.13}}
 {id=WavePlayer_0, properties {soundWaveAssetPtr=&/Glider_PineTree_Open}}
]
properties {
 attenuationSettings=/Game/Sounds/Attenuation/Fort_Player_Glider_OpenClose_Athena_Att
 audioPropertiesSheet=/BRCosmetics/Sounds/Settings/PropertySheets/Glider/APS_Cosmetic_Glider_Open
 maxDistance=10000
 soundClassObject=&Fort_Gliders_Land
 volumeMultiplier=0.9
 concurrencySet [&Athena_GliderOpenClose_Conc]
}
